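# Looping sound cue for the scan beam sight of a Heist sentry, built from Security Turret and camera scan beam audio.
# A mixer blends two looping waves, a narrow beam at reduced volume and a wide beam at full volume, each with its own cone attenuation asset.
# The cue plays in the gadgets sound class at boosted volume, with overridden concurrency that caps simultaneous instances and stops the quietest.
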SoundCue /HeistBasesGameplay/Audio/Sentry/Heist_Sentury_ScanBeam_Sight_Loop_Cue
root=Mixer_1
nodes [
 {
  id=Attenuation_0
  children [WavePlayer_7]
  properties {attenuationSettings=/Game/Athena/Sounds/Gadgets/SecurityTurret/Security_Turret_Sight_Beam_Cone_Narrow_Att}
 }
 {
  id=Attenuation_4
  children [WavePlayer_5]
  properties {attenuationSettings=/Game/Athena/Sounds/Gadgets/SecurityTurret/Security_Turret_Sight_Beam_Cone_Wide_Att}
 }
 {id=Mixer_1, children [Attenuation_0, Attenuation_4], properties {inputVolume [0.6,1]}}
 {
  id=WavePlayer_5
  properties {bLooping=true, soundWaveAssetPtr=/Game/Athena/Sounds/Gadgets/SecurityTurret/Security_Camera_ScanBeam_Loop_01_Wide}
 }
 {
  id=WavePlayer_7
  properties {bLooping=true, soundWaveAssetPtr=/Game/Athena/Sounds/Gadgets/SecurityTurret/Security_Camera_ScanBeam_Loop_01_Narrow}
 }
]
properties {
 bOverrideConcurrency=true
 maxDistance=2550
 soundClassObject=&Fort_Gadgets
 volumeMultiplier=2
 attenuationOverrides {
  bEnableFocusInterpolation=true
  bEnableListenerFocus=true
  falloffDistance=1200
  nonFocusVolumeAttenuation=0.5
  stereoSpread=500
  attenuationShapeExtents {x=0}
 }
 concurrencyOverrides {maxCount=3, resolutionRule=StopQuietest}
}
